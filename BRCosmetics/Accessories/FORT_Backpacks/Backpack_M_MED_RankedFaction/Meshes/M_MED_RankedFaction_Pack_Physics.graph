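# Physics asset for the Ranked Faction backpack cosmetic mesh, giving its dyn_backpack bones simple collision bodies.
# Three differently sized spheres cover the base, lower and upper bones, with the base and upper spheres offset from their bones.
# Lower and upper each constrain to the base with swing and twist locked, so the set 45-degree limits allow no relative rotation.
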PhysicsAsset /BRCosmetics/Accessories/FORT_Backpacks/Backpack_M_MED_RankedFaction/Meshes/M_MED_RankedFaction_Pack_Physics
bodies {
 dyn_backpack_base [{shapeType=Sphere, radius=8.5, shapeName=dyn_backpack_base_sphere, center=<0 1.9334 0>}]
 dyn_backpack_lower [{shapeType=Sphere, radius=3.5, shapeName=dyn_backpack_lower_sphere}]
 dyn_backpack_upper [{shapeType=Sphere, radius=7.5, shapeName=dyn_backpack_upper_sphere, center=<0 -1.1377 2.732>}]
}
constraints [
 {
  bone1Name=dyn_backpack_lower
  bone2Name=dyn_backpack_base
  swing1LimitDegrees=45
  swing1Motion=Locked
  swing2LimitDegrees=45
  swing2Motion=Locked
  twistLimitDegrees=45
  twistMotion=Locked
 }
 {
  bone1Name=dyn_backpack_upper
  bone2Name=dyn_backpack_base
  swing1LimitDegrees=45
  swing1Motion=Locked
  swing2LimitDegrees=45
  swing2Motion=Locked
  twistLimitDegrees=45
  twistMotion=Locked
 }
]
