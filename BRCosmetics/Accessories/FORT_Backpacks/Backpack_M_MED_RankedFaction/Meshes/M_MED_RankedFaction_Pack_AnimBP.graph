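# Animation Blueprint for the MED RankedFaction backpack cosmetic, giving its base, upper and lower bones a floating sway.
# Each update, real-time sine waves drive per-bone X/Z offsets, with a negative Y shift while skydiving before the parachute opens.
# The AnimGraph nudges the pack bone, applies those offsets additively in bone space, then layers partial-weight RigidBody and spring-based AnimDynamics on the base bone.
# Several EventGraph math nodes are unconnected leftovers.
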AnimBlueprint /BRCosmetics/Accessories/FORT_Backpacks/Backpack_M_MED_RankedFaction/Meshes/M_MED_RankedFaction_Pack_AnimBP
target_skeleton=&/M_MED_RankedFaction_Pack_Skeleton
graphs [
 AnimGraph {nodes [
   n1 = "Transform (Modify) Bone - Bone: pack"(ComponentPose=Local To Component(LocalPose=Input Pose()), Translation="0,3,0", Rotation="0,0,0") {translationMode=BMM_Additive, alphaBoolBlend {blendOption=Linear}, boneToModify {boneName=pack}, translation {y=3}}
   n2 = "Transform (Modify) Bone - Bone: dyn_backpack_base"(ComponentPose=@n1, Alpha=1, Translation=Get Dyn_pack_base() -> (Dyn_pack_base:Vector), Rotation="0,0,0", Scale="1,1,1") {translationMode=BMM_Additive, translationSpace=BCS_BoneSpace, alphaBoolBlend {blendOption=Linear}, boneToModify {boneName=dyn_backpack_base}}
   n3 = "Transform (Modify) Bone - Bone: dyn_backpack_upper"(ComponentPose=@n2, Alpha=0.75, Translation=Get Dyn_pack_upper() -> (Dyn_pack_upper:Vector), Rotation="0,0,0", Scale="1,1,1") {
    translationMode=BMM_Additive
    translationSpace=BCS_BoneSpace
    alphaBoolBlend {blendOption=Linear}
    boneToModify {boneName=dyn_backpack_upper}
   }
   n4 = "Transform (Modify) Bone - Bone: dyn_backpack_lower"(ComponentPose=@n3, Alpha=0.75, Translation=Get Dyn_pack_lower() -> (Dyn_pack_lower:Vector), Rotation="0,0,0", Scale="1,1,1") {
    translationMode=BMM_Additive
    translationSpace=BCS_BoneSpace
    alphaBoolBlend {blendOption=Linear}
    boneToModify {boneName=dyn_backpack_lower}
   }
   n5 = RigidBody(ComponentPose=@n4, Alpha=0.5) {
    bOverrideWorldGravity=true
    overlapChannel=ECC_PhysicsBody
    overridePhysicsAsset=&/M_MED_RankedFaction_Pack_Physics
    alphaBoolBlend {blendOption=Linear}
    baseBoneRef {boneName=root}
    componentAppliedLinearAccClamp=<10 10 10>
    componentLinearAccScale=<1 0.1 1>
    componentLinearVelScale=<1 1 1>
    simSpaceSettings {maxAngularAcceleration=0, maxAngularVelocity=0, maxLinearAcceleration=50, maxLinearVelocity=50, worldAlpha=0.5}
   }
   n6 = Anim Dynamics - Bone: dyn_backpack_base(ComponentPose=@n5, Alpha=0.75, ExternalForce="0,0,0") {
    angularDampingOverride=0.7
    bLinearSpring=true
    bOverrideAngularDamping=true
    bOverrideLinearDamping=true
    bUseGravityOverride=true
    linearDampingOverride=1
    linearSpringConstant=60
    alphaBoolBlend {blendOption=Linear}
    boundBone {boneName=dyn_backpack_base}
    componentAppliedLinearAccClamp=<500 500 500>
    componentLinearAccScale=<1 1 1>
    componentLinearVelScale=<1 1 1>
    physicsBodyDefinitions [{
      boundBone {boneName=dyn_backpack_base}
      boxExtents=<10 10 10>
      constraintSetup {angularLimitsMax=<0 0 0>, angularLimitsMin=<0 0 0>, angularTarget=<0 0 0>, linearAxesMax=<5 5 5>, linearAxesMin=<-5 0 -5>}
      localJointOffset=<0 0 0>
     }]
   }
   Output Pose(Result=Component To Local(ComponentPose=@n6)) {layerGroup=None}
  ]}
 EventGraph {nodes [
   Event BlueprintUpdateAnimation()
   > n1 = Sequence!(execute) -> (then_0, then_1, then_2:Exec)
   n2 = Get bIsParachuteOpened()
   > n3 = Branch(Condition=AND Boolean(A=Get bIsSkyDiving(), B=NOT Boolean(A=@n2)))
   n4 = GetRealTimeSeconds()
   n5 = float * float(A=-0.55, B=1.25)
   n6 = float * float(A="Sin (Radians)"(A=float * float(A=@n5, B=@n4)), B=0.85)
   n7 = float * float(A=1.36, B=0.75)
   n8 = float * float(A="Sin (Radians)"(A=float * float(A=@n7, B=@n4)), B=0.25)
   > Set Dyn_pack_upper(Dyn_pack_upper:Vector=MakeVector(X=@n8, Y=-2.5, Z=@n6)) -> (Output_Get:Vector)
   Set Dyn_pack_upper(execute=@n3.else, Dyn_pack_upper=MakeVector(X=@n8, Y=0, Z=@n6))
   n9 = Get bIsParachuteOpened()
   n10 = Branch(execute=@n1.then_1, Condition=AND Boolean(A=Get bIsSkyDiving(), B=NOT Boolean(A=@n9)))
   n11 = GetRealTimeSeconds()
   n12 = float * float(A=1.36, B=0.75)
   n13 = float * float(A="Sin (Radians)"(A=float * float(A=@n12, B=@n11)), B=0.8)
   n14 = float * float(A=-0.55, B=1.25)
   n15 = float * float(A="Sin (Radians)"(A=float * float(A=@n14, B=@n11)), B=1.2)
   > Set Dyn_pack_base(Dyn_pack_base:Vector=MakeVector(X=@n13, Y=-2.5, Z=@n15)) -> (Output_Get:Vector)
   Set Dyn_pack_base(execute=@n10.else, Dyn_pack_base=MakeVector(X=@n13, Y=0, Z=@n15))
   n16 = Get bIsParachuteOpened()
   n17 = Branch(execute=@n1.then_2, Condition=AND Boolean(A=Get bIsSkyDiving(), B=NOT Boolean(A=@n16)))
   n18 = GetRealTimeSeconds()
   n19 = float * float(A=1.36, B=0.85)
   n20 = float * float(A="Sin (Radians)"(A=float * float(A=@n19, B=@n18)), B=0.3)
   n21 = float * float(A=float * float(A=-0.85, B=0.65), B=@n18)
   n22 = float + float(A=-0.5, B=float * float(A="Sin (Radians)"(A=@n21), B=0.65))
   > Set Dyn_pack_lower(Dyn_pack_lower:Vector=MakeVector(X=@n20, Y=-7.5, Z=@n22)) -> (Output_Get:Vector)
   Set Dyn_pack_lower(execute=@n17.else, Dyn_pack_lower=MakeVector(X=@n20, Y=0, Z=@n22))
   TryGetPawnOwner()
   n23 = float * float(A=1.11, B=1.11)
   float * float(A="Sin (Radians)"(A=float * float(A=@n23, B=@n11)), B=50)
   n24 = float * float(A=1.11, B=1.11)
   float * float(A="Sin (Radians)"(A=float * float(A=@n24, B=@n4)), B=50)
   n25 = float * float(A=1.11, B=1.11)
   float * float(A="Sin (Radians)"(A=float * float(A=@n25, B=@n18)), B=1)
  ]}
]
variables [
 Dyn_pack_base {type=Vector}
 Dyn_pack_lower {type=Vector}
 Dyn_pack_upper {type=Vector}
]
